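# Translucent unlit blueprint material for player-built walls, glowing with a tiled paper texture and channel-selected outlines, with a cheaper mobile selection path.
# World position offset sums hit deformation, a per-board break animation from SelfBuildingWalls pivot data, and a sine dance bounce, plus a camera offset.
# Opacity height-lerps by the code opacity and break phase, fading near the camera and scene intersections.
# Custom primitive data drives animation, opacity, emissive color and interact brightness.
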Material /Game/Packages/Fortnite_UniqueMaterials_1/Walls/PlayerBuiltWall_BluePrintShader
graph {
 properties {
  bUsedWithNanite=true
  bUsedWithStaticLighting=true
  blendMode=BLEND_Translucent
  pixelDepthOffsetMode=PDOM_Legacy
  shadingModel=MSM_Unlit
  usedShadingModels=Unlit
 }
 nodes [
  n1 = MaterialFunctionCall!(Number of sub objects, Manual Anim Grad, Sec mot Falloff rad) {materialFunction=/Game/Effects/Fort_Effects/Materials/Temp/SelfBuildingWalls}
  n2 = Frac(None=@n1.Board Pivot Pos)
  n3 = ScalarParameter() {bUseCustomPrimitiveData=true, group=Code, parameterName=code_BreakAnimation, primitiveDataIndex=5}
  n4 = Subtract(A=@n1.Board Pivot Pos, B=ObjectPositionWS())
  n5 = LinearInterpolate(A=CrossProduct(A=Normalize(VectorInput=@n4), B=@n1.Board Orientation), B=@n1.Board Orientation)
  n6 = ConstantBiasScale(None=@n1.Random Per Mesh) {bias=-0.5, scale=1}
  n7 = Add(
   A=Multiply(A=ConstantBiasScale(None=@n2) {bias=-0.5, scale=256}, B=@n3) {constB=50}
   B=RotateAboutAxis(
    NormalizedRotationAxis=Normalize(VectorInput=@n5)
    RotationAngle=Multiply(A=@n3, B=@n6)
    PivotPoint=@n1.Board Pivot Pos
    Position=WorldPosition().XYZ
   )
  )
  n8 = Constant()
  n9 = WorldPosition()
  MaterialExpressionObjectWorldPosition_1 = ObjectPositionWS()
  n10 = Add(A=Subtract(A=@n9.XYZ, B=@MaterialExpressionObjectWorldPosition_1), B=Constant3Vector() {constant=<rgba 0 0 64 0>})
  n11 = Normalize(VectorInput=@n10)
  n12 = ObjectRadius()
  n13 = Subtract(A=@n12, B=ComponentMask(None=Abs(None=@n10)) {b=true})
  n14 = Sine(None=Time())
  n15 = Multiply(A=Multiply(A=Constant() {r=-0.128}, B=Multiply(A=@n11, B=@n13)), B=@n14)
  n16 = Constant() {desc=magnitude of vertical bounce, r=0.125}
  n17 = Subtract(A=@n9.XYZ, B=Subtract(A=@MaterialExpressionObjectWorldPosition_1, B=@n12))
  n18 = AppendVector(A=Constant2Vector(), B=Multiply(A=Multiply(A=@n14, B=@n16), B=ComponentMask(None=@n17) {b=true}))
  n19 = Multiply(
   A=ScalarParameter() {bUseCustomPrimitiveData=true, group=Code, parameterName=code_JonsDanceTime, primitiveDataIndex=7}
   B=Multiply(A=Add(A=@n15, B=@n18), B=Constant() {r=0.0503})
  )
  n20 = Add(
   A=MaterialFunctionCall!(
    WSDeformationCenter=VectorParameter() {bUseCustomPrimitiveData=true, group=z_code, parameterName=WSDeformationCenter, primitiveDataIndex=20}.RGB
    Code_DestructionAmount=ScalarParameter() {group=z_code, parameterName=Code_DestructionAmount}
    Code_UseUvForDamage
    Code_UseHitEffects=StaticBoolParameter() {defaultValue=true, group=z_code, parameterName=Code_UseHitEffects}
    Code_Emissive
    WSDamageRadius=ScalarParameter() {bUseCustomPrimitiveData=true, group=z_code, parameterName=WSDamageRadius, primitiveDataIndex=26}
    CrackDamageShader
    Diffuse
    ScalarModDamageCracks
    ScalarOffsetforDamageCracks
    ExposeLerpFunctionForInvertingTheHitDistanceCalc
    InvertDamageFalloff
    CrackWorldPosition
    WorldPosition
    DistanceCullFadeAnimationOffset
   ) {materialFunction=/Game/Packages/jl_tests/Fortnite_HitEffects}.Hit And Shrink WPO
   B=Add(
    A=StaticSwitchParameter(True=@n7, False=@n8) {defaultValue=true, group=Code, parameterName=Enable Death Animation}
    B=StaticSwitchParameter(True=@n19, False=@n8) {defaultValue=true, group=Code, parameterName=Enable Jons Dance Time}
   )
  )
  n21 = Subtract(A=Add(A=@n20, B=WorldPosition().XYZ), B=CameraPositionWS())
  Add(A=@n20, B=Multiply(A=Normalize(VectorInput=@n21), B=ScalarParameter() {defaultValue=5, group=Code, parameterName=Camera offset amount}))
  n22 = Multiply(A=ScalarParameter() {bUseCustomPrimitiveData=true, defaultValue=1, parameterName=code_opacity, primitiveDataIndex=6}, B=OneMinus(None=@n3))
  n23 = Clamp(None=NamedRerouteUsage())
  n24 = TextureSample(UVs=Multiply(
    A=TextureCoordinate() {uTiling=2, vTiling=2}
    B=ScalarParameter() {defaultValue=1, group=Blueprint_Info, parameterName=Blueprint_GridMultiplier}
   )) {texture=/Game/Packages/Fortnite_UniqueMaterials_1/Textures/paper2}
  n25 = Constant() {r=0.5}
  n26 = LinearInterpolate(A=Multiply(A=@n24.RGB, B=@n24.RGB), B=@n25, Alpha=@n25)
  n27 = ComponentMask(None=@n26) {r=true}
  n28 = PixelDepth()
  n29 = Constant()
  n30 = Subtract(A=SceneDepth().R, B=@n28.R)
  n31 = Multiply(
   A=OneMinus(None=SphereMask(A=@n28.R, B=@n29) {attenuationRadius=128, hardnessPercent=10})
   B=OneMinus(None=SphereMask(A=@n30, B=@n29) {attenuationRadius=16, hardnessPercent=10})
  )
  n32 = Multiply(A=@n31, B=@n31)
  n33 = Clamp(None=Multiply(A=@n32, B=@n32))
  Multiply(
   A=Multiply(
    A=MaterialFunctionCall(Transition Phase=@n22, Height Texture=@n23) {materialFunction=&HeightLerp}.Alpha
    B=Multiply(A=@n27, B=@n33)
   ) {desc=Opacity}
   B=ScalarParameter() {defaultValue=1, parameterName=OpacityMultiplier}
  )
  n34 = ComponentMask(None=TextureCoordinate() {coordinateIndex=1}) {r=true}
  n35 = TextureSampleParameter2D() {group=Blueprint_Info, parameterName=Blueprint_Outline_Texture, texture=/Game/Packages/Fortress_SharedMaterials/Textures/bluePrintOutlines1}
  n36 = If(A=@n34, B=Constant() {r=1}, "A > B"=@n35.G, "A == B"=@n35.G, "A < B"=@n35.G)
  n37 = If(A=@n34, B=Constant() {r=2}, "A > B"=@n35.B, "A == B"=@n36, "A < B"=@n36)
  n38 = TextureSampleParameter2D() {group=Outline, parameterName=Blueprint_Outline_Texture, texture=&bluePrintOutlines1}
  n39 = AppendVector(A=@n38.RGB, B=@n38.RGB)
  n40 = VectorParameter() {group=Blueprint_Info, parameterName=Blueprint_Outline_Texture_Channel, defaultValue=<rgba 1 0 0 0>}
  n41 = VectorParameter() {group=Blueprint_Info, parameterName=Blueprint_Outline_Texture_Channel2, defaultValue=<rgba 0 1 0 0>}
  n42 = VectorParameter() {group=Blueprint_Info, parameterName=Blueprint_Outline_Texture_Channel3, defaultValue=<rgba 0 0 1 0>}
  n43 = VectorParameter() {group=Blueprint_Info, parameterName=Blueprint_Outline_Texture_Channel4}
  n44 = MaterialFunctionCall!(
   Spec1
   Texture1=@n39
   Normal1
   Normal2
   Texture2=@n39
   Spec2
   Texture3=@n39
   Spec3
   Normal3
   Texture4=@n39
   Spec4
   Normal4
   Texture5
   Normal5
   Spec5
   Texture6
   Spec6
   Normal6
   Texture7
   Texture8
   Texture9
   Use_Dot_Texture9
   Texture10
   Use 2 Diffuse Textures
   --------------
   Use 3 Diffuse Textures
   Use 4 Diffuse Textures=StaticBoolParameter() {defaultValue=true, group=Multiple_Material_Textures, parameterName=Use 4 Materials}
   Use 5 Diffuse Textures
   Use 6 Diffuse Textures
   Use 7 Diffuse Textures
   Use 8 Diffuse Textures
   Use 9 Diffuse Textures
   Use 10 Diffuse Textures
   "-------------- "
   Dot_Texture1=AppendVector(A=@n40.RGB, B=@n40.RGB)
   TransitionInfo=ComponentMask(None=TextureCoordinate() {uTiling=0.5}) {r=true}
   Dot_Texture2=AppendVector(A=@n41.RGB, B=@n41.RGB)
   Dot_Texture3=AppendVector(A=@n42.RGB, B=@n42.RGB)
   Dot_Texture4=AppendVector(A=@n43.RGB, B=@n43.RGB)
   Dot_Texture5
   Dot_Texture6
   Dot_Texture7
   Dot_Texture8
   Dot_Texture9
   Dot_Texture10
   Use_Dot_Texture1=StaticBoolParameter() {defaultValue=true, group=Multiple_Material_Textures, parameterName=Dot Texture 1}
   Use_Dot_Texture2=StaticBoolParameter() {defaultValue=true, group=Multiple_Material_Textures, parameterName=Dot Texture 2}
   Use_Dot_Texture3=StaticBoolParameter() {defaultValue=true, group=Multiple_Material_Textures, parameterName=Dot Texture 3}
   Use_Dot_Texture4=StaticBoolParameter() {defaultValue=true, group=Multiple_Material_Textures, parameterName=Dot Texture 4}
   Use_Dot_Texture5
   Use_Dot_Texture6
   Use_Dot_Texture7
   Use_Dot_Texture8
   Use_Dot_Texture10
  ) {materialFunction=/Game/Packages/Fortress_SharedMaterials/Base_Material_Functions/Fortnite_TextureSetLerp}
  NamedRerouteDeclaration(None=StaticSwitchParameter(True=@n37, False=@n44.Diffuse) {parameterName=useMobilePath}) {name=Outline, nodeColor=<rgba 0 1 0.3906 1>}
  n45 = Multiply(A=Constant(), B=Fresnel())
  n46 = VectorParameter() {group=Code, parameterName=Code_Color, defaultValue=<rgba 0 0.059 0.3923 1>}
  n47 = VectorParameter() {desc=Don't delete, group=Code, parameterName=Code_Red, defaultValue=<rgba 0.6464 0.073 0.0335 0.2376>}
  n48 = Clamp(None=Subtract(
    A=MaterialFunctionCall!(Number of sub objects, Manual Anim Grad, Sec mot Falloff rad) {materialFunction=/Game/Effects/Fort_Effects/Materials/Temp/SelfBuildingWalls2}.Anim Order
    B=ScalarParameter() {defaultValue=1000, group=Build, parameterName=TargetBreakNumber}
   ))
  n49 = VectorParameter() {bUseCustomPrimitiveData=true, parameterName=Code_Emissive, defaultValue=<rgba 0 0 0 1>}
  n50 = ComponentMask(None=@n49.RGB) {r=true}
  n51 = StaticSwitchParameter(True=LinearInterpolate(A=@n46.R, B=@n47.R, Alpha=@n48), False=LinearInterpolate(A=@n46.RGB, B=@n49.RGB, Alpha=Clamp(None=@n50))) {group=test, parameterName=TestBrokenMaterials}
  n52 = VectorParameter() {desc=Don't delete, group=Code, parameterName=Code_Green, defaultValue=<rgba 0.1965 0.6132 0.2751 0.2376>}
  n53 = Add(A=VectorParameter() {desc=Don't delete, group=Code, parameterName=code_Blue, defaultValue=<rgba 0 0.059 0.3923 1>}.RGB, B=@n52.RGB)
  n54 = Multiply(
   A=Add(
    A=Add(A=@n47.RGB, B=@n53)
    B=VectorParameter() {desc=Don't delete, group=Code, parameterName=Code_Orange, defaultValue=<rgba 1 0.4996 0 0.6011>}.RGB
   )
   B=Constant()
  )
  n55 = Add(A=Multiply(A=Add(A=@n51, B=@n54), B=@n27), B=Multiply(A=ComponentMask(None=@n26) {g=true}, B=Constant() {r=0.125}))
  n56 = NamedRerouteUsage()
  n57 = Add(A=@n55, B=Multiply(A=@n56) {constB=2})
  n58 = LinearInterpolate(A=@n57, B=Multiply(A=@n55, B=OneMinus(None=@n56)), Alpha=@n48)
  n59 = Add(
   A=Multiply(A=Fresnel() {exponent=1}, B=NamedRerouteUsage())
   B=Add(A=Multiply(A=@n27, B=@n45), B=StaticSwitchParameter(True=@n58, False=@n57) {group=test, parameterName=TestBrokenMaterials})
  )
  n60 = Multiply(
   A=@n59
   B=Multiply(
    A=ScalarParameter() {bUseCustomPrimitiveData=true, parameterName=Interact, primitiveDataIndex=8}
    B=ScalarParameter() {defaultValue=1, parameterName=PunchBrightness}
   )
  )
  n61 = MaterialFunctionCall(Emissive=Clamp(None=Add(A=@n59, B=@n60)) {maxDefault=2.75, minDefault=-10}) {materialFunction=&MF_FX_EmissiveAdjustments}
  RayTracingQualitySwitch(Normal=@n61.Result, RayTraced=Saturate(None=@n61.Result))
  ShadingPathSwitch(Default=@n44.Diffuse, Mobile=@n37) {desc=Migrated from FeatureLevelSwitch}
  ShadingPathSwitch(Default=@n44.Diffuse, Mobile=DistanceFieldsRenderingSwitch!(No=@n37, Yes=@n44.Diffuse) {desc="Switch renderer=true"})
 ]
}
